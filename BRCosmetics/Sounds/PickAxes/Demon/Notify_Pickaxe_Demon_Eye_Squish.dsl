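; Anim notify for the Demon pickaxe cosmetic's eye squish, adding music ducking on top of the parent notify's behavior.
; Received_Notify casts the mesh component's owner to B_Melee_Impact_Pickaxe_Athena_Demon, then calls the parent implementation.
; It then pushes the Demon pickaxe DuckMusic sound mix modifier and returns false.
; The cast result is not otherwise used, and the mix is not popped here.
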(fn Received_Notify
  (params (MeshComp :type Skeletal%20Mesh%20Component%20Object%20Reference) (Animation :type Anim%20Sequence%20Base%20Object%20Reference) (EventReference :type Anim%20Notify%20Event%20Reference%20Structure%20%28by%20ref%29))
  (Utilities|Casting|CastToB_Melee_Impact_Pickaxe_Athena_Demon (Components|GetOwner MeshComp))
  (|CallParentFunction MeshComp Animation EventReference)
  (Audio|PushSoundMixModifier "/Game/Sounds/FortSoundClassesAndModes/Fort_Pickaxe_Demon_DuckMusic_MixMod.Fort_Pickaxe_Demon_DuckMusic_MixMod")
  (return false))
