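# Sound cue for the Demon pickaxe's eye squish, a quiet short-range effect in the melee weapon sound class.
# A quality-level node picks one of three modulated random branches, with the last two levels sharing a two-wave branch.
# The first level draws on five waves at reduced volume, and the second on three with wider pitch variation.
# It primes on load and overrides attenuation, enabling occlusion traces.
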SoundCue /BRCosmetics/Sounds/PickAxes/Demon/Pickaxe_Demon_Eye_Squish
root=QualityLevel_0
nodes [
 {id=Modulator_0, children [Random_2], properties {pitchMax=1.02, pitchMin=0.98, volumeMax=0.35, volumeMin=0.1}}
 {id=Modulator_1, children [Random_0]}
 {id=Modulator_2, children [Random_4], properties {pitchMax=1.15, pitchMin=0.8}}
 {id=QualityLevel_0, children [Modulator_0,Modulator_2,Modulator_1,Modulator_1]}
 {id=Random_0, children [WavePlayer_1,WavePlayer_2], properties {weights [1,1]}}
 {id=Random_2, children [WavePlayer_0, WavePlayer_22, WavePlayer_23, WavePlayer_26, WavePlayer_27], properties {weights [1,1,1,1,1]}}
 {id=Random_4, children [WavePlayer_29, WavePlayer_30, WavePlayer_3], properties {weights [1,1,1]}}
 {id=WavePlayer_0, properties {soundWaveAssetPtr=&/Demon_Pickaxe_Eye_Squish_09}}
 {id=WavePlayer_1, properties {soundWaveAssetPtr=&/Demon_Pickaxe_Eye_Squish_07}}
 {id=WavePlayer_2, properties {soundWaveAssetPtr=&Demon_Pickaxe_Eye_Squish_09}}
 {id=WavePlayer_22, properties {soundWaveAssetPtr=&/Demon_Pickaxe_Eye_Squish_03}}
 {id=WavePlayer_23, properties {soundWaveAssetPtr=&/Demon_Pickaxe_Eye_Squish_04}}
 {id=WavePlayer_26, properties {soundWaveAssetPtr=&Demon_Pickaxe_Eye_Squish_07}}
 {id=WavePlayer_27, properties {soundWaveAssetPtr=&/Demon_Pickaxe_Eye_Squish_08}}
 {id=WavePlayer_29, properties {soundWaveAssetPtr=&Demon_Pickaxe_Eye_Squish_07}}
 {id=WavePlayer_3, properties {soundWaveAssetPtr=&Demon_Pickaxe_Eye_Squish_04}}
 {id=WavePlayer_30, properties {soundWaveAssetPtr=&Demon_Pickaxe_Eye_Squish_09}}
]
properties {
 bOverrideAttenuation=true
 bPrimeOnLoad=true
 maxDistance=750
 soundClassObject=&Fort_WeaponMelee
 volumeMultiplier=0.25
 attenuationOverrides {
  bEnableOcclusion=true
  falloffDistance=400
  focusAzimuth=15
  nonFocusAzimuth=20
  nonFocusVolumeAttenuation=0.2
  occlusionInterpolationTime=0.5
  occlusionLowPassFilterFrequency=60000
  occlusionTraceChannel=ECC_GameTraceChannel11
  occlusionVolumeAttenuation=0
  attenuationShapeExtents {x=350}
 }
}
