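# Sound cue for the impact of the Demon pickaxe cosmetic, layering recycled Dragon impact sounds with demon growls.
# A mixer blends three briefly delayed, modulated layers, each picking a random wave: Dragon hammer impacts at full volume, with quieter Dragon pickaxe impacts and Demon growl impacts.
# It uses pickaxe impact attenuation, property sheet and melee weapon concurrency, and primes on load.
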SoundCue /BRCosmetics/Sounds/PickAxes/Demon/Pickaxe_Demon_Impact
root=Mixer_1
nodes [
 {id=Delay_0, children [Modulator_1], properties {delayMax=0.1, delayMin=0.1}}
 {id=Delay_1, children [Modulator_0], properties {delayMax=0.1, delayMin=0.1}}
 {id=Delay_2, children [Modulator_3], properties {delayMax=0.1, delayMin=0.1}}
 {id=Mixer_1, children [Delay_1,Delay_0,Delay_2], properties {inputVolume [1,0.5,0.6]}}
 {id=Modulator_0, children [Random_0], properties {pitchMax=1.1, pitchMin=1}}
 {id=Modulator_1, children [Random_3], properties {pitchMax=1, pitchMin=1}}
 {id=Modulator_3, children [Random_2]}
 {id=Random_0, children [WavePlayer_12, WavePlayer_13], properties {weights [1,1]}}
 {id=Random_2, children [WavePlayer_7, WavePlayer_8, WavePlayer_9, WavePlayer_11], properties {weights [1,1,1,1]}}
 {id=Random_3, children [WavePlayer_4, WavePlayer_10, WavePlayer_3], properties {weights [1,1,1]}}
 {id=WavePlayer_10, properties {soundWaveAssetPtr=/BRCosmetics/Sounds/PickAxes/PA_Dragon_Impact_02}}
 {id=WavePlayer_11, properties {soundWaveAssetPtr=&/Demon_Pickaxe_Growl_Impact_10}}
 {id=WavePlayer_12, properties {soundWaveAssetPtr=/Game/Sounds/Fort_Weapons/Dragon/Hammer/Dragon_Hammer_Impact_01}}
 {id=WavePlayer_13, properties {soundWaveAssetPtr=/Game/Sounds/Fort_Weapons/Dragon/Hammer/Dragon_Hammer_Impact_02}}
 {id=WavePlayer_3, properties {soundWaveAssetPtr=/BRCosmetics/Sounds/PickAxes/PA_Dragon_Impact_04}}
 {id=WavePlayer_4, properties {soundWaveAssetPtr=/BRCosmetics/Sounds/PickAxes/PA_Dragon_Impact_01}}
 {id=WavePlayer_7, properties {soundWaveAssetPtr=&/Demon_Pickaxe_Growl_Impact_11}}
 {id=WavePlayer_8, properties {soundWaveAssetPtr=&/Demon_Pickaxe_Growl_Impact_12}}
 {id=WavePlayer_9, properties {soundWaveAssetPtr=&/Demon_Pickaxe_Growl_Impact_13}}
]
properties {
 attenuationSettings=/Game/Sounds/Attenuation/Fort_Pickaxe_Impact_Att
 audioPropertiesSheet=/BRCosmetics/Sounds/Settings/PropertySheets/Pickaxe/APS_Cosmetic_Pickaxe_Impact
 bPrimeOnLoad=true
 maxDistance=4300
 soundClassObject=&Fort_Impacts
 concurrencyOverrides {bLimitToOwner=true, maxCount=2, resolutionRule=StopOldest}
 concurrencySet [&Player_Melee_Weapon_Conc]
}
