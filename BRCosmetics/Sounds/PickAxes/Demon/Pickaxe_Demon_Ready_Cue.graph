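# Sound cue for the Demon cosmetic pickaxe's ready sound, mixing close and medium-range swing layers.
# The close layer picks randomly from Demon swing waves with wider pitch variation, while the medium layer picks from generic distant fire axe swings, each under its own pickaxe attenuation.
# A root modulator adds slight pitch variation to the mix, and the cue primes on load under the melee weapon sound class.
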SoundCue /BRCosmetics/Sounds/PickAxes/Demon/Pickaxe_Demon_Ready_Cue
root=Modulator_1
nodes [
 {id=Attenuation_2, children [Modulator_7], properties {attenuationSettings=/Game/Sounds/Attenuation/Fort_Weapon_Pickaxe_Close_Att}}
 {id=Attenuation_3, children [Random_0], properties {attenuationSettings=/Game/Sounds/Attenuation/Fort_Weapon_Pickaxe_Medium_Att}}
 {id=Mixer_1, children [Attenuation_2, Attenuation_3], properties {inputVolume [1,1]}}
 {id=Modulator_1, children [Mixer_1], properties {pitchMax=1.03, pitchMin=0.97}}
 {id=Modulator_7, children [Random_5], properties {pitchMax=1.1, pitchMin=0.9}}
 {
  id=Random_0
  children [WavePlayer_13, WavePlayer_1, WavePlayer_2, WavePlayer_12, WavePlayer_4, WavePlayer_16]
  properties {weights [1,1,1,1,1,1]}
 }
 {id=Random_5, children [WavePlayer_25, WavePlayer_26, WavePlayer_27, WavePlayer_29], properties {weights [1,1,1,1]}}
 {id=WavePlayer_1, properties {soundWaveAssetPtr=/Game/Athena/Sounds/Weapons/Fire_Axe_Swing_Distant_01}}
 {id=WavePlayer_12, properties {soundWaveAssetPtr=/Game/Athena/Sounds/Weapons/Fire_Axe_Swing_Distant_03}}
 {id=WavePlayer_13, properties {soundWaveAssetPtr=/Game/Athena/Sounds/Weapons/Fire_Axe_Swing_Distant_04}}
 {id=WavePlayer_16, properties {soundWaveAssetPtr=/Game/Athena/Sounds/Weapons/Fire_Axe_Swing_Distant_06}}
 {id=WavePlayer_2, properties {soundWaveAssetPtr=/Game/Athena/Sounds/Weapons/Fire_Axe_Swing_Distant_02}}
 {id=WavePlayer_25, properties {soundWaveAssetPtr=&/Demon_Swing_01}}
 {id=WavePlayer_26, properties {soundWaveAssetPtr=&/Demon_Swing_02}}
 {id=WavePlayer_27, properties {soundWaveAssetPtr=&/Demon_Swing_03}}
 {id=WavePlayer_29, properties {soundWaveAssetPtr=&/Demon_Swing_04}}
 {id=WavePlayer_4, properties {soundWaveAssetPtr=/Game/Athena/Sounds/Weapons/Fire_Axe_Swing_Distant_05}}
]
properties {
 audioPropertiesSheet=/BRCosmetics/Sounds/Settings/PropertySheets/Pickaxe/APS_Cosmetic_Pickaxe_Ready
 bPrimeOnLoad=true
 maxDistance=3000
 soundClassObject=&Fort_WeaponMelee
 volumeMultiplier=0.7
}
